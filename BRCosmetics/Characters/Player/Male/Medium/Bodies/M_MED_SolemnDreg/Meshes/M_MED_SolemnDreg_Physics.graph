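# Physics asset for the SolemnDreg medium male player body, set up for secondary motion on its jacket tails.
# Capsules on the pelvis and left thigh give body volume, alongside a back jacket chain and a two-segment front-left chain.
# Both chains hang from tiny base capsules constrained to the pelvis, with narrow angular limits along the dynamic segments.
# The left thigh and front-left base rotate freely, while the back base locks twist.
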PhysicsAsset /BRCosmetics/Characters/Player/Male/Medium/Bodies/M_MED_SolemnDreg/Meshes/M_MED_SolemnDreg_Physics
bodies {
 Base_jacket_back [{shapeType=Capsule, length=0.505, radius=0.505, shapeName=Base_jacket_back_capsule, rotation=<rot 0 -90 0>}]
 Base_jacket_front_l [{shapeType=Capsule, length=0.505, radius=0.505, shapeName=Base_jacket_front_l_capsule}]
 dyn_jacket_back [{
   shapeType=Capsule
   length=14.1456
   radius=5.8928
   shapeName=dyn_jacket_back_capsule
   center=<5.9952 -4.1068 1.8709>
   rotation=<rot 4.7621 -88.2111 -4.3963>
  }]
 dyn_jacket_front_l_1 [{shapeType=Capsule, length=12.7809, radius=5.2304, shapeName=dyn_jacket_front_l_1_capsule, center=<0 2.654 0>, rotation=<rot 0 0.0005 0>}]
 dyn_jacket_front_l_2 [{shapeType=Capsule, length=7.3932, radius=5.6901, shapeName=dyn_jacket_front_l_2_capsule, center=<3.2886 4.2613 0>, rotation=<rot 0 0.0006 5>}]
 pelvis [{shapeType=Capsule, length=9.8212, radius=10.2259, shapeName=pelvis_capsule, center=<0 -1.0535 -0.0249>, rotation=<rot -1.3541 -90 -3.3178>}]
 thigh_l [{
   shapeType=Capsule
   length=17.1084
   radius=10.7615
   shapeName=thigh_l_capsule
   center=<-9.7485 0.514 -0.1136>
   rotation=<rot 66.9464 -94.5847 -90.4976>
  }]
}
constraints [
 {
  bone1Name=thigh_l
  bone2Name=pelvis
  swing1LimitDegrees=45
  swing1Motion=Free
  swing2LimitDegrees=45
  swing2Motion=Free
  twistLimitDegrees=45
  twistMotion=Free
 }
 {bone1Name=Base_jacket_back, bone2Name=pelvis, swing1LimitDegrees=30, swing2LimitDegrees=3, twistLimitDegrees=45, twistMotion=Locked}
 {bone1Name=dyn_jacket_back, bone2Name=Base_jacket_back, swing1LimitDegrees=45, swing2LimitDegrees=5, twistLimitDegrees=5}
 {
  bone1Name=Base_jacket_front_l
  bone2Name=pelvis
  swing1LimitDegrees=30
  swing1Motion=Free
  swing2LimitDegrees=10
  swing2Motion=Free
  twistLimitDegrees=5
  twistMotion=Free
 }
 {bone1Name=dyn_jacket_front_l_1, bone2Name=Base_jacket_front_l, swing1LimitDegrees=30, swing2LimitDegrees=10, twistLimitDegrees=5}
 {bone1Name=dyn_jacket_front_l_2, bone2Name=dyn_jacket_front_l_1, swing1LimitDegrees=30, swing2LimitDegrees=10, twistLimitDegrees=5}
]
