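# Post-copy AnimBP for the SolemnDreg medium male body cosmetic, layering correctives and secondary motion onto the attached parent's pose.
# The output chain runs a knee-corrective constraint, a Control Rig fed DBNO, crouch and knee-slide state, jacket AnimDynamics, then pelvis-space RigidBody.
# Shoulder constraints and parachute/clavicle-driven offsets form a chain disconnected from the output.
# The EventGraph still updates clavicle alphas, plus left-leg and gravity values the AnimGraph never reads.
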AnimBlueprint /BRCosmetics/Characters/Player/Male/Medium/Bodies/M_MED_SolemnDreg/Meshes/M_MED_SolemnDreg_AnimBP
target_skeleton=&/M_MED_SolemnDreg_Skeleton
graphs [
 AnimGraph {nodes [
   n1 = "Constraint - dyn_shoulder_r (P)"(Alpha=1, ConstraintWeights_0=1) {
    alphaBoolBlend {blendOption=Linear}
    boneToModify {boneName=dyn_shoulder_r}
    constraintSetup [{targetBone {boneName=clavicle_r}}]
    constraintWeights [1]
   }
   n2 = "Constraint - dyn_shoulder_l (P)"(ComponentPose=@n1, Alpha=1, ConstraintWeights_0=1) {
    alphaBoolBlend {blendOption=Linear}
    boneToModify {boneName=dyn_shoulder_l}
    constraintSetup [{targetBone {boneName=clavicle_l}}]
    constraintWeights [1]
   }
   n3 = Get bIsParachuteOpened()
   n4 = "Transform (Modify) Bone - Bone: dyn_shoulder_r"!(ComponentPose=@n2, bAlphaBoolEnabled:bool=@n3, Translation="-2,0,2", Rotation="-15.000,0.000,0.000", Scale="1,1,1") -> (Pose) {
    alphaInputType=Bool
    rotationMode=BMM_Additive
    rotationSpace=BCS_BoneSpace
    translationMode=BMM_Additive
    translationSpace=BCS_BoneSpace
    alphaBoolBlend {blendInTime=1, blendOption=Linear, blendOutTime=1}
    boneToModify {boneName=dyn_shoulder_r}
    rotation {pitch=-15}
    translation {x=-2, z=2}
   }
   n5 = "Transform (Modify) Bone - Bone: dyn_shoulder_l"!(ComponentPose=@n4, bAlphaBoolEnabled:bool=@n3, Translation="-1,0,-1.25", Rotation="12,0.000,0.000", Scale="1,1,1") -> (Pose) {
    alphaInputType=Bool
    rotationMode=BMM_Additive
    rotationSpace=BCS_BoneSpace
    translationMode=BMM_Additive
    translationSpace=BCS_BoneSpace
    alphaBoolBlend {blendInTime=1, blendOption=Linear, blendOutTime=1}
    boneToModify {boneName=dyn_shoulder_l}
    rotation {pitch=12}
    translation {x=-1, z=-1.25}
   }
   n6 = "Transform (Modify) Bone - Bone: dyn_shoulder_r"!(
    ComponentPose=@n5
    Alpha:float=Get Clav R Rotation Alpha() -> (Clav R Rotation Alpha:double)
    Translation="-2,0,1.5"
    Rotation="-15,0,0.000"
    Scale="1,1,1"
   ) -> (Pose) {
    rotationMode=BMM_Additive
    rotationSpace=BCS_BoneSpace
    translationMode=BMM_Additive
    translationSpace=BCS_BoneSpace
    alphaBoolBlend {blendInTime=1, blendOption=Linear, blendOutTime=1}
    boneToModify {boneName=dyn_shoulder_r}
    rotation {pitch=-15}
    translation {x=-2, z=1.5}
   }
   "Transform (Modify) Bone - Bone: dyn_shoulder_l"!(
    ComponentPose=@n6
    Alpha:float=Get Clav L Rotation Alpha() -> (Clav L Rotation Alpha:double)
    Translation="-2,0,-1.5"
    Rotation="15,0.000,0.000"
    Scale="1,1,1"
   ) -> (Pose) {
    rotationMode=BMM_Additive
    rotationSpace=BCS_BoneSpace
    translationMode=BMM_Additive
    translationSpace=BCS_BoneSpace
    alphaBoolBlend {blendInTime=1, blendOption=Linear, blendOutTime=1}
    boneToModify {boneName=dyn_shoulder_l}
    rotation {pitch=15}
    translation {x=-2, z=-1.5}
   }
   n7 = "Constraint - Base_knee_corr (PPPPPPPP)"!(
    ComponentPose=Local To Component(LocalPose=Copy Pose From Mesh() {bUseAttachedParent=true})
    Alpha=1
    ConstraintWeights_0=0.2234
    ConstraintWeights_1:float=0.2132
    ConstraintWeights_2:float=0.0294
    ConstraintWeights_3:float=0.0485
    ConstraintWeights_4:float=0.0454
    ConstraintWeights_5:float=0.0202
    ConstraintWeights_6:float=0
    ConstraintWeights_7:float=0.42
   ) -> (Pose) {
    alphaBoolBlend {blendOption=Linear}
    boneToModify {boneName=Base_knee_corr}
    constraintSetup [
     {targetBone {boneName=thigh_l}}
     {targetBone {boneName=calf_l}}
     {targetBone {boneName=calf_twist_02_l}}
     {targetBone {boneName=deform_knee_out_l}}
     {targetBone {boneName=deform_knee_in_l}}
     {targetBone {boneName=thigh_twist_01_l}}
     {targetBone {boneName=deform_groin_l}}
     {targetBone {boneName=Dyn_knee_corr}}
    ]
    constraintWeights [1,1,1,1,1,1,1,1]
   }
   n8 = Control Rig!(
    Source=Component To Local(ComponentPose=@n7)
    Alpha
    dbno:bool=Get bIsDBNO()
    KneeSlide:double=Get Slide_Alpha() -> (Slide_Alpha:double)
    IsCrouch:bool=Get bIsCrouching()
   ) -> (Pose) {
    alphaBoolBlend {blendOption=Linear}
    controlRigAssetReference {blueprintRigClass=&/M_MED_SolemnDreg_CtrlRig.M_MED_SolemnDreg_CtrlRig_C}
   }
   n9 = Anim Dynamics - Bone: dyn_jacket_front_r_1(ComponentPose=Local To Component(LocalPose=@n8), Alpha=1, ExternalForce="0,0,0") {
    angularDampingOverride=0.95
    angularSpringConstant=300
    bAngularSpring=true
    bGravityOverrideInSimSpace=true
    bLinearSpring=true
    bOverrideAngularDamping=true
    bUseGravityOverride=true
    gravityScale=0.35
    linearSpringConstant=50
    numSolverIterationsPostUpdate=2
    numSolverIterationsPreUpdate=8
    alphaBoolBlend {blendOption=Linear}
    boundBone {boneName=dyn_jacket_front_r_1}
    componentAppliedLinearAccClamp=<5000 5000 5000>
    componentLinearAccScale=<0.5 0.5 0.5>
    componentLinearVelScale=<0.5 0.5 5>
    gravityOverride {z=-980}
    physicsBodyDefinitions [{
      collisionType=CustomSphere
      sphereCollisionRadius=4
      boundBone {boneName=dyn_jacket_front_r_1}
      boxExtents=<12 12 12>
      constraintSetup {angularLimitsMax=<10 20 15>, angularLimitsMin=<-10 -1 -4>, angularTarget=<1 0 0>}
      localJointOffset=<1 0 0>
     }]
    sphericalLimits [{limitRadius=5, drivingBone {boneName=thigh_r}, sphereLocalOffset=<7.6962 -2.8316 1.2847>}]
   }
   n10 = RigidBody!(ComponentPose=@n9, Alpha, OverrideWorldGravity:Vector="0,0,-980", ExternalForce) -> (Pose) {
    bOverrideWorldGravity=true
    overridePhysicsAsset=&/M_MED_SolemnDreg_Physics
    simulationSpace=BaseBoneSpace
    alphaBoolBlend {blendOption=Linear}
    baseBoneRef {boneName=pelvis}
    componentAppliedLinearAccClamp=<3000 3000 3000>
    overrideWorldGravity {z=-980}
    simSpaceSettings {
     dampingAlpha=0.8
     maxAngularAcceleration=40
     maxAngularVelocity=8
     maxLinearAcceleration=3000
     maxLinearVelocity=700
     velocityScaleZ=0.8
     worldAlpha=0.8
    }
   }
   Output Pose(Result=Component To Local(ComponentPose=@n10)) {layerGroup=None}
  ]}
 EventGraph {nodes [
   Event BlueprintUpdateAnimation()
   > n1 = Sequence()
   n2 = Get bIsParachuteOpened()
   n3 = Branch(execute=@n1.then_1, Condition=@n2)
   > Set Clav R Rotation Alpha(Clav R Rotation Alpha:double=0) -> (Output_Get:double)
   n4 = BreakTransform(InTransform=GetDeltaTransformFromRefPose(self=GetOwnerSkeletalMeshComponent(), BoneName=upperarm_r, BaseName=spine_05))
   Set Clav R Rotation Alpha(execute=@n3.else, Clav R Rotation Alpha=MapRangeClamped(Value=BreakRotator(InRot=@n4.Rotation).Pitch, InRangeA=50, InRangeB=70))
   > n5 = Branch(Condition=@n2)
   > Set Clav L Rotation Alpha(Clav L Rotation Alpha:double=0) -> (Output_Get:double)
   n6 = BreakTransform(InTransform=GetDeltaTransformFromRefPose(self=GetOwnerSkeletalMeshComponent(), BoneName=upperarm_l, BaseName=spine_05))
   Set Clav L Rotation Alpha(execute=@n5.else, Clav L Rotation Alpha=MapRangeClamped(Value=BreakRotator(InRot=@n6.Rotation).Pitch, InRangeA=50, InRangeB=70))
   Set LeftLegForwardBendCompensationAlpha(
    execute=@n1.then_0
    LeftLegForwardBendCompensationAlpha:double=GetAlphaFromReferenceTransformDelta(
     SkeletalMeshComponent=Get Sminem Skel Mesh() -> (Sminem Skel Mesh:Skeletal Mesh Component Object Reference)
     BoneName=thigh_l
     BaseName=pelvis
     InputRange=(X=-17,Y=-103)
     TwistAxis="1,0,0"
    )
   ) -> (Output_Get:double)
   > n7 = Gravity Override State Function DEPRECATED(CharacterPartAnimInstance=Self-Reference(), GameTimeInSeconds=GetGameTimeInSeconds(), GravityOverrideParams=Get Gravity Override Params())
   > Set Gravity Override(Gravity Override=@n7)
   Event OnBeginKneeSliding() {is=title K2Node_Event}
   > Set Slide_Alpha(Slide_Alpha:double=1) -> (Output_Get:double)
   Event BlueprintInitializeAnimation()
   n8 = GetOwnerSkeletalMeshComponent()
   > Set MeshToCopy(MeshToCopy:Skeletal Mesh Component Object Reference=@n8) -> (Output_Get:Skeletal Mesh Component Object Reference)
   > Set Sminem Skel Mesh(Sminem Skel Mesh:Skeletal Mesh Component Object Reference=@n8) -> (Output_Get:Skeletal Mesh Component Object Reference)
   Event OnEndKneeSliding() {is=title K2Node_Event}
   > Set Slide_Alpha(Slide_Alpha=0)
  ]}
]
variables [
 Clav L Rotation Alpha {type=number}
 Clav R Rotation Alpha {type=number}
 Gravity Override {type=Vector}
 Gravity Override Params {type=GravityOverrideParamsStruct}
 LeftLegForwardBendCompensationAlpha {type=number}
 MeshToCopy {type=SkeletalMeshComponent}
 Slide_Alpha {type=number}
 Sminem Skel Mesh {type=SkeletalMeshComponent}
]
