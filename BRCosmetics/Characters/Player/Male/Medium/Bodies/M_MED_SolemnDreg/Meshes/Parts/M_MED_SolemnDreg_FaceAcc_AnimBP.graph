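# Animation Blueprint for the SolemnDreg male medium character's face accessory mesh, adding secondary motion to the front hair.
# It copies the pose from the attached parent mesh, then runs a single Anim Dynamics node on Dyn_hair_front before output.
# That root-relative simulation uses linear and angular springs, high damping, gravity, angular constraints and two spherical limits driven by the left thigh bone.
# The event graph holds no logic.
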AnimBlueprint /BRCosmetics/Characters/Player/Male/Medium/Bodies/M_MED_SolemnDreg/Meshes/Parts/M_MED_SolemnDreg_FaceAcc_AnimBP
target_skeleton=&/M_MED_SolemnDreg_FaceAcc_Skeleton
graphs [
 AnimGraph {nodes [
   n1 = Anim Dynamics - Bone: Dyn_hair_front!(
    ComponentPose=Local To Component(LocalPose=Copy Pose From Mesh() {bUseAttachedParent=true})
    bAlphaBoolEnabled:bool=True
    ExternalForce="0,0,0"
   ) -> (Pose) {
    alphaInputType=Bool
    angularDampingOverride=0.999
    angularSpringConstant=300
    bAngularSpring=true
    bLinearSpring=true
    bOverrideAngularDamping=true
    bOverrideLinearDamping=true
    bUseGravityOverride=true
    bUseSphericalLimits=true
    linearDampingOverride=0.999
    linearSpringConstant=50
    simulationSpace=RootRelative
    alphaBoolBlend {blendOption=Linear}
    boundBone {boneName=Dyn_hair_front}
    componentAppliedLinearAccClamp=<7000 7000 7000>
    componentLinearAccScale=<8 8 8>
    componentLinearVelScale=<8 8 8>
    gravityOverride {z=-980}
    physicsBodyDefinitions [{
      collisionType=CustomSphere
      sphereCollisionRadius=4
      boundBone {boneName=Dyn_hair_front}
      boxExtents=<11 11 11>
      constraintSetup {angularLimitsMax=<3 0 20>, angularLimitsMin=<-3 -25 -7>, angularTarget=<1 0 0>, linearAxesMax=<0 0 1>}
      localJointOffset=<0.5 0 0>
     }]
    relativeSpaceBone {boneName=pelvis}
    sphericalLimits [
     {limitRadius=4.963, drivingBone {boneName=thigh_l}, sphereLocalOffset=<-8.1225 2.3763 0>}
     {limitRadius=4.3261, drivingBone {boneName=thigh_l}, sphereLocalOffset=<-10.4115 3.1899 4.7015>}
    ]
   }
   Output Pose(Result=Component To Local(ComponentPose=@n1)) {layerGroup=None}
  ]}
 EventGraph {nodes [Event BlueprintUpdateAnimation()]}
]
variables []
